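# Physics asset for the jacket of the PersimmonSmoke male medium player cosmetic, providing ragdoll-style secondary motion for the garment.
# Capsules on pelvis, spine and thighs form a body proxy joined by locked constraints.
# Five jacket chains (front left/right, both sides, back) hang from tiny base anchors locked to spine_03, swinging and twisting within tight angular limits.
# A small zipper body dangles from the end of the left front chain.
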PhysicsAsset /BRCosmetics/Characters/Player/Male/Medium/Bodies/M_MED_PersimmonSmoke/Meshes/M_MED_PersimmonSmoke_Physics_Jacket
bodies {
 Base_Jacket_01_Back [{shapeType=Capsule, length=0.505, radius=0.505, shapeName=Base_Jacket_01_Back_capsule, rotation=<rot 0 -90 0>}]
 Base_Jacket_01_l [{shapeType=Capsule, length=0.505, radius=0.505, shapeName=Base_Jacket_01_l_capsule, rotation=<rot 0 -90 0>}]
 Base_Jacket_01_r [{shapeType=Capsule, length=0.505, radius=0.505, shapeName=Base_Jacket_01_r_capsule, rotation=<rot 0 -90 0>}]
 Base_Jacket_01_side_l [{shapeType=Capsule, length=0.505, radius=0.505, shapeName=Base_Jacket_01_side_l_capsule, rotation=<rot 0 -90 0>}]
 Base_Jacket_01_side_r [{shapeType=Capsule, length=0.505, radius=0.505, shapeName=Base_Jacket_01_side_r_capsule, rotation=<rot 0 -90 0>}]
 Dyn_zipper [{shapeType=Capsule, length=2.9, radius=1.7, center=<1.5909 0 0.0522>, rotation=<rot -88.1214 0.0002 -0.0001>}]
 Jacket_01__Back [{shapeType=Capsule, length=13.5, radius=3.4, center=<1.2025 0 1.3369>, rotation=<rot 0 0 90>}]
 Jacket_01_l [{shapeType=Capsule, length=8.4, radius=4.3, center=<1 0 3.5>, rotation=<rot 0 0 90>}]
 Jacket_01_r [{shapeType=Capsule, length=9.3, radius=4.3, center=<1 0 3.5>, rotation=<rot 0 0 90>}]
 Jacket_01_side_l [{shapeType=Capsule, length=13.5, radius=3.4, center=<1.2025 0 1.3369>, rotation=<rot 0 0 90>}]
 Jacket_01_side_r [{shapeType=Capsule, length=13.5, radius=3.4, center=<1.2025 0 1.3369>, rotation=<rot 0 0 90>}]
 Jacket_02__Back [{shapeType=Capsule, length=10.6, radius=6.3, center=<1.2025 0 1.3369>, rotation=<rot 0 0 90>}]
 Jacket_02_l [{shapeType=Capsule, length=10.2, radius=4.3, center=<1 0 3.5>, rotation=<rot 0 0 90>}]
 Jacket_02_r [{shapeType=Capsule, length=9.3, radius=4.3, center=<1 0 3.5>, rotation=<rot 0 0 90>}]
 Jacket_02_side_l [{shapeType=Capsule, length=13.5, radius=3.4, center=<1.2025 0 1.3369>, rotation=<rot 0 0 90>}]
 Jacket_02_side_r [{shapeType=Capsule, length=13.5, radius=3.4, center=<1.2025 0 1.3369>, rotation=<rot 0 0 90>}]
 Jacket_03_l [{shapeType=Capsule, length=13.4, radius=4.4, center=<1 0 3.5>, rotation=<rot 0 0 86.6506>}]
 Jacket_03_r [{shapeType=Capsule, length=9.1, radius=4.5, center=<1 0 3.5>, rotation=<rot 0 0 95.5136>}]
 pelvis [{shapeType=Capsule, length=7.7339, radius=11.0561, shapeName=pelvis_capsule, center=<2.3854 -0.6852 0.0943>}]
 spine_01 [{shapeType=Capsule, length=12.7, radius=7.8}]
 spine_02 [{shapeType=Capsule, length=8.8, radius=10}]
 spine_03 [{shapeType=Capsule, length=11, radius=8.9}]
 thigh_l [{shapeType=Capsule, length=24.4, radius=8.9, center=<-10.3782 0.8464 0.0733>, rotation=<rot 89.5954 0.0014 -179.9985>}]
 thigh_r [{shapeType=Capsule, length=24.4, radius=8.9, center=<10.3782 -0.8464 -0.0733>, rotation=<rot 89.5954 0.0014 0.0015>}]
}
constraints [
 {
  bone1Name=spine_01
  bone2Name=pelvis
  swing1LimitDegrees=45
  swing1Motion=Locked
  swing2LimitDegrees=45
  swing2Motion=Locked
  twistLimitDegrees=45
  twistMotion=Locked
 }
 {
  bone1Name=spine_02
  bone2Name=spine_01
  swing1LimitDegrees=45
  swing1Motion=Locked
  swing2LimitDegrees=45
  swing2Motion=Locked
  twistLimitDegrees=45
  twistMotion=Locked
 }
 {
  bone1Name=spine_03
  bone2Name=spine_02
  swing1LimitDegrees=45
  swing1Motion=Locked
  swing2LimitDegrees=45
  swing2Motion=Locked
  twistLimitDegrees=45
  twistMotion=Locked
 }
 {
  bone1Name=thigh_l
  bone2Name=pelvis
  swing1LimitDegrees=45
  swing1Motion=Locked
  swing2LimitDegrees=45
  swing2Motion=Locked
  twistLimitDegrees=45
  twistMotion=Locked
 }
 {
  bone1Name=thigh_r
  bone2Name=pelvis
  swing1LimitDegrees=45
  swing1Motion=Locked
  swing2LimitDegrees=45
  swing2Motion=Locked
  twistLimitDegrees=45
  twistMotion=Locked
 }
 {bone1Name=Jacket_02_side_r, bone2Name=Jacket_01_side_r, swing1LimitDegrees=10, swing2LimitDegrees=10, twistLimitDegrees=5}
 {bone1Name=Jacket_01__Back, bone2Name=Base_Jacket_01_Back, swing1LimitDegrees=2, swing2LimitDegrees=10, twistLimitDegrees=4}
 {bone1Name=Jacket_02__Back, bone2Name=Jacket_01__Back, swing1LimitDegrees=2, swing2LimitDegrees=10, twistLimitDegrees=4}
 {
  bone1Name=Base_Jacket_01_l
  bone2Name=spine_03
  swing1LimitDegrees=45
  swing1Motion=Locked
  swing2LimitDegrees=45
  swing2Motion=Locked
  twistLimitDegrees=45
  twistMotion=Locked
 }
 {bone1Name=Jacket_01_l, bone2Name=Base_Jacket_01_l, swing1LimitDegrees=10, swing2LimitDegrees=20, twistLimitDegrees=5}
 {bone1Name=Jacket_02_l, bone2Name=Jacket_01_l, swing1LimitDegrees=10, swing2LimitDegrees=20, twistLimitDegrees=5}
 {bone1Name=Jacket_03_l, bone2Name=Jacket_02_l, swing1LimitDegrees=10, swing2LimitDegrees=25, twistLimitDegrees=5}
 {
  bone1Name=Base_Jacket_01_r
  bone2Name=spine_03
  swing1LimitDegrees=45
  swing1Motion=Locked
  swing2LimitDegrees=45
  swing2Motion=Locked
  twistLimitDegrees=45
  twistMotion=Locked
 }
 {bone1Name=Jacket_01_r, bone2Name=Base_Jacket_01_r, swing1LimitDegrees=10, swing2LimitDegrees=20, twistLimitDegrees=5}
 {bone1Name=Jacket_02_r, bone2Name=Jacket_01_r, swing1LimitDegrees=10, swing2LimitDegrees=20, twistLimitDegrees=5}
 {bone1Name=Jacket_03_r, bone2Name=Jacket_02_r, swing1LimitDegrees=10, swing2LimitDegrees=20, twistLimitDegrees=5}
 {
  bone1Name=Base_Jacket_01_side_l
  bone2Name=spine_03
  swing1LimitDegrees=45
  swing1Motion=Locked
  swing2LimitDegrees=45
  swing2Motion=Locked
  twistLimitDegrees=45
  twistMotion=Locked
 }
 {bone1Name=Jacket_01_side_l, bone2Name=Base_Jacket_01_side_l, swing1LimitDegrees=5, swing2LimitDegrees=10, twistLimitDegrees=5}
 {bone1Name=Jacket_02_side_l, bone2Name=Jacket_01_side_l, swing1LimitDegrees=10, swing2LimitDegrees=10, twistLimitDegrees=5}
 {
  bone1Name=Base_Jacket_01_side_r
  bone2Name=spine_03
  swing1LimitDegrees=45
  swing1Motion=Locked
  swing2LimitDegrees=45
  swing2Motion=Locked
  twistLimitDegrees=45
  twistMotion=Locked
 }
 {bone1Name=Jacket_01_side_r, bone2Name=Base_Jacket_01_side_r, swing1LimitDegrees=5, swing2LimitDegrees=10, twistLimitDegrees=5}
 {
  bone1Name=Base_Jacket_01_Back
  bone2Name=spine_03
  swing1LimitDegrees=45
  swing1Motion=Locked
  swing2LimitDegrees=45
  swing2Motion=Locked
  twistLimitDegrees=45
  twistMotion=Locked
 }
 {bone1Name=Dyn_zipper, bone2Name=Jacket_03_l, swing1LimitDegrees=10, swing2LimitDegrees=20, twistLimitDegrees=5}
]
