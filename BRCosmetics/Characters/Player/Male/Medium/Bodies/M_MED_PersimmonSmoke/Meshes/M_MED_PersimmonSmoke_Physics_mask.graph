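# Physics asset for the mask accessory on the PersimmonSmoke medium male player body, which is anchored at the pelvis.
# Four capsule bodies cover pelvis, right thigh, Base_Mask and dyn_Mask.
# Base_Mask has swing and twist locked to the pelvis, and dyn_Mask hangs off it with tighter second-swing and twist angles than first swing.
# The right thigh also joins the pelvis, giving a leg volume alongside the mask.
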PhysicsAsset /BRCosmetics/Characters/Player/Male/Medium/Bodies/M_MED_PersimmonSmoke/Meshes/M_MED_PersimmonSmoke_Physics_mask
bodies {
 Base_Mask [{shapeType=Capsule, length=0.505, radius=0.505, shapeName=Base_Mask_capsule, rotation=<rot 0 -90 0>}]
 dyn_Mask [{
   shapeType=Capsule
   length=13.6953
   radius=5.7479
   shapeName=dyn_Mask_capsule
   center=<7.8317 -1.5257 4.295>
   rotation=<rot -0.003 81.5039 -91.3677>
  }]
 pelvis [{shapeType=Capsule, length=13.9813, radius=7.4287}]
 thigh_r [{shapeType=Capsule, length=19.4141, radius=10.4553, center=<12.2994 -0.5795 -3.1092>, rotation=<rot 0 -92.4029 -92.6321>}]
}
constraints [
 {bone1Name=thigh_r, bone2Name=pelvis, swing1LimitDegrees=45, swing2LimitDegrees=45, twistLimitDegrees=45}
 {bone1Name=dyn_Mask, bone2Name=Base_Mask, swing1LimitDegrees=45, swing2LimitDegrees=9, twistLimitDegrees=10}
 {
  bone1Name=Base_Mask
  bone2Name=pelvis
  swing1LimitDegrees=45
  swing1Motion=Locked
  swing2LimitDegrees=45
  swing2Motion=Locked
  twistLimitDegrees=45
  twistMotion=Locked
 }
]
