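# Physics asset for the PersimmonSmoke male medium character's face accessory mesh, set up for constrained physical motion of its hair bones.
# Capsule bodies cover the hair bones, head and upper spine (spine_05), with head constrained to the spine.
# Most hair bones hang directly off the head, while Hair_01/02, Hair_03/04 and Hair_14/15 form two-link chains.
# Hair limits allow very little twist and favor swing2 over swing1, which is locked on Hair_02.
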PhysicsAsset /BRCosmetics/Characters/Player/Male/Medium/Bodies/M_MED_PersimmonSmoke/Meshes/Parts/M_MED_PersimmonSmoke_FaceAcc_Physics
bodies {
 Hair_01 [{shapeType=Capsule, length=0.505, radius=0.505, shapeName=Hair_01_capsule, rotation=<rot 0 -90 0>}]
 Hair_02 [{
   shapeType=Capsule
   length=1.6463
   radius=2.1217
   shapeName=Hair_02_capsule
   center=<1.8672 0.1726 -0.8382>
   rotation=<rot 0 -74.7519 -62.8866>
  }]
 Hair_03 [{shapeType=Capsule, length=1.1293, radius=0.9858, shapeName=Hair_03_capsule, center=<2.8264 -0.0478 0.329>, rotation=<rot 0 0.5737 -85.4896>}]
 Hair_04 [{
   shapeType=Capsule
   length=1.9939
   radius=1.3309
   shapeName=Hair_04_capsule
   center=<1.5468 -0.2173 0.5218>
   rotation=<rot -14.0642 -90 -36.058>
  }]
 Hair_05 [{shapeType=Capsule, length=1.736, radius=1.424, shapeName=Hair_05_capsule, center=<3.5173 0.0765 -0.0708>, rotation=<rot 0 -87.952 -86.8529>}]
 Hair_06 [{
   shapeType=Capsule
   length=3.1026
   radius=1.9767
   shapeName=Hair_06_capsule
   center=<5.6463 -0.5314 -0.66>
   rotation=<rot 0 -105.0832 -57.0839>
  }]
 Hair_08 [{
   shapeType=Capsule
   length=3.6873
   radius=1.7647
   shapeName=Hair_08_capsule
   center=<4.5354 0.4402 0.418>
   rotation=<rot -18.3687 -103.505 -79.2862>
  }]
 Hair_10 [{
   shapeType=Capsule
   length=4.0318
   radius=3.0322
   shapeName=Hair_10_capsule
   center=<3.9672 0.4607 0.0618>
   rotation=<rot -9.2959 -104.6746 -66.2529>
  }]
 Hair_12 [{
   shapeType=Capsule
   length=1.801
   radius=1.3516
   shapeName=Hair_12_capsule
   center=<5.8436 -0.0555 -0.603>
   rotation=<rot 0 -89.1157 -78.1805>
  }]
 Hair_14 [{
   shapeType=Capsule
   length=1.7737
   radius=1.3133
   shapeName=Hair_14_capsule
   center=<1.5307 0.6745 1.6273>
   rotation=<rot 0 -144.0376 -61.6815>
  }]
 Hair_15 [{
   shapeType=Capsule
   length=2.8134
   radius=1.6859
   shapeName=Hair_15_capsule
   center=<0.4822 -0.54 -5.2097>
   rotation=<rot 0 -89.6511 -38.8532>
  }]
 Hair_16 [{shapeType=Capsule, length=1.8229, radius=1.5282, shapeName=Hair_16_capsule, center=<2.86 0.2557 -0.3928>, rotation=<rot 0 -157.978 -77.3115>}]
 Hair_17 [{shapeType=Capsule, length=2.4319, radius=1.7321, shapeName=Hair_17_capsule, center=<5.4688 0.3355 0.3619>, rotation=<rot 0 -89.12 -89.6051>}]
 Hair_18 [{shapeType=Capsule, length=1.9871, radius=1.712, shapeName=Hair_18_capsule, center=<5.5135 0.4439 0.1641>, rotation=<rot 0 -84.7139 -92.9542>}]
 Hair_19 [{shapeType=Capsule, length=3.2678, radius=2.2017, shapeName=Hair_19_capsule, center=<4.5 0 0>, rotation=<rot -18.4781 82.9971 -103.1024>}]
 Hair_21 [{shapeType=Capsule, length=0.505, radius=0.505, shapeName=Hair_21_capsule, rotation=<rot 0 -90 0>}]
 Hair_22 [{
   shapeType=Capsule
   length=1.5515
   radius=1.4196
   shapeName=Hair_22_capsule
   center=<3.6168 0.1188 -0.3365>
   rotation=<rot -22.5948 -90 -29.3212>
  }]
 Hair_25 [{
   shapeType=Capsule
   length=3.2845
   radius=1.7709
   shapeName=Hair_25_capsule
   center=<3.5897 1.2127 0.5714>
   rotation=<rot -35.7943 -90 -27.1967>
  }]
 Hair_28 [{
   shapeType=Capsule
   length=3.0314
   radius=2.1391
   shapeName=Hair_28_capsule
   center=<3.5093 -0.0113 0.399>
   rotation=<rot 14.3518 -94.0504 -86.0522>
  }]
 head [{shapeType=Capsule, length=10.3328, radius=8.0992, center=<-1.0648 1.9596 -0.0308>, rotation=<rot 12.6809 -97.1474 -90>}]
 spine_05 [{shapeType=Capsule, length=1, radius=1}, {shapeType=Capsule, length=15, radius=15, shapeName=spine_05_capsule, center=<5.2928 0 0>}]
}
constraints [
 {bone1Name=head, bone2Name=spine_05, swing1LimitDegrees=45, swing2LimitDegrees=45, twistLimitDegrees=45}
 {bone1Name=Hair_01, bone2Name=head, swing1LimitDegrees=3, swing2LimitDegrees=20, twistLimitDegrees=2}
 {bone1Name=Hair_02, bone2Name=Hair_01, swing1LimitDegrees=10, swing1Motion=Locked, swing2LimitDegrees=20, twistLimitDegrees=2}
 {bone1Name=Hair_03, bone2Name=head, swing1LimitDegrees=4, swing2LimitDegrees=25, twistLimitDegrees=2}
 {bone1Name=Hair_04, bone2Name=Hair_03, swing1LimitDegrees=4, swing2LimitDegrees=10, twistLimitDegrees=2}
 {bone1Name=Hair_05, bone2Name=head, swing1LimitDegrees=5, swing2LimitDegrees=25, twistLimitDegrees=2}
 {bone1Name=Hair_06, bone2Name=head, swing1LimitDegrees=8.64, swing2LimitDegrees=30, twistLimitDegrees=2}
 {bone1Name=Hair_08, bone2Name=head, swing1LimitDegrees=5, swing2LimitDegrees=25, twistLimitDegrees=2}
 {bone1Name=Hair_10, bone2Name=head, swing1LimitDegrees=6, swing2LimitDegrees=20, twistLimitDegrees=2}
 {bone1Name=Hair_12, bone2Name=head, swing1LimitDegrees=3, swing2LimitDegrees=20, twistLimitDegrees=2}
 {bone1Name=Hair_14, bone2Name=head, swing1LimitDegrees=4, swing2LimitDegrees=10, twistLimitDegrees=2}
 {bone1Name=Hair_15, bone2Name=Hair_14, swing1LimitDegrees=4, swing2LimitDegrees=25, twistLimitDegrees=2}
 {bone1Name=Hair_16, bone2Name=head, swing1LimitDegrees=3, swing2LimitDegrees=25, twistLimitDegrees=2}
 {bone1Name=Hair_17, bone2Name=head, swing1LimitDegrees=5, swing2LimitDegrees=15, twistLimitDegrees=2}
 {bone1Name=Hair_18, bone2Name=head, swing1LimitDegrees=5, swing2LimitDegrees=25, twistLimitDegrees=2}
 {bone1Name=Hair_19, bone2Name=head, swing1LimitDegrees=4, swing2LimitDegrees=30, twistLimitDegrees=2}
 {bone1Name=Hair_21, bone2Name=head, swing1LimitDegrees=5.6, swing2LimitDegrees=15, twistLimitDegrees=2}
 {bone1Name=Hair_22, bone2Name=head, swing1LimitDegrees=4, swing2LimitDegrees=15, twistLimitDegrees=2}
 {bone1Name=Hair_25, bone2Name=head, swing1LimitDegrees=5, swing2LimitDegrees=10, twistLimitDegrees=2}
 {bone1Name=Hair_28, bone2Name=head, swing1LimitDegrees=10, swing2LimitDegrees=20, twistLimitDegrees=2}
]
